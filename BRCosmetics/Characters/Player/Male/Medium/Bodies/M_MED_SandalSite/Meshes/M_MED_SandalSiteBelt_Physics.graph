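# Physics asset for the belt mesh of the SandalSite male medium body cosmetic, defining collision and joints for its dangling belt stripes.
# Two capsule chains hang from the MSTR_BeltStripes sphere, three links left and two right, with tight twist limits and swing narrowing below the first joint.
# Both second links carry an extra angled capsule.
# Unconstrained pelvis, pelvis-corrective, thigh and thigh-twist capsules provide body shapes alongside the chains.
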PhysicsAsset /BRCosmetics/Characters/Player/Male/Medium/Bodies/M_MED_SandalSite/Meshes/M_MED_SandalSiteBelt_Physics
bodies {
 MSTR_BeltStripes [{shapeType=Sphere, radius=1.005, shapeName=MSTR_BeltStripes_sphere}]
 dyn_beltStripe01_l [{
   shapeType=Capsule
   length=3.4775
   radius=1.274
   shapeName=dyn_beltStripe01_l_capsule
   center=<2.4836 -0.1662 -0.0222>
   rotation=<rot -0.5132 -89.5254 -86.0348>
  }]
 dyn_beltStripe01_r [{
   shapeType=Capsule
   length=4.1757
   radius=1.2591
   shapeName=dyn_beltStripe01_r_capsule
   center=<-3.1874 0.2986 -0.1412>
   rotation=<rot -0.577 -88.6858 -85.6734>
  }]
 dyn_beltStripe02_l [
  {
   shapeType=Capsule
   length=4.8453
   radius=1.6597
   shapeName=dyn_beltStripe02_l_capsule
   center=<2.8075 -0.0637 -0.0325>
   rotation=<rot -5.3867 -89.6568 87.9772>
  }
  {
   shapeType=Capsule
   length=5.5
   radius=1.5
   shapeName=dyn_beltStripe02_l_capsule
   center=<0.8796 -0.2084 -1.9967>
   rotation=<rot 76.2461 -0.0001 -2.7915>
  }
 ]
 dyn_beltStripe02_r [
  {
   shapeType=Capsule
   length=5.2113
   radius=1.7586
   shapeName=dyn_beltStripe02_r_capsule
   center=<-4.2191 0.1688 -0.5258>
   rotation=<rot 0 90.438 -88.5125>
  }
  {
   shapeType=Capsule
   length=4.5
   radius=1.5
   shapeName=dyn_beltStripe02_r_capsule
   center=<-0.9525 0.0185 1.8372>
   rotation=<rot 81.1369 -28.9323 -31.5407>
  }
 ]
 dyn_beltStripe03_l [{
   shapeType=Capsule
   length=3.1829
   radius=1.9371
   shapeName=dyn_beltStripe03_l_capsule
   center=<3.2158 0.0869 0.1851>
   rotation=<rot 0 91.0518 -83.3862>
  }]
 dyn_pelvisCorrColl [{
   shapeType=Capsule
   length=3.5674
   radius=5.6987
   shapeName=dyn_pelvisCorrColl_capsule
   center=<-0.0002 -1.1333 -0.0118>
   rotation=<rot -0.0001 -90 90.977>
  }]
 pelvis [{shapeType=Capsule, length=9.3469, radius=11.7559, shapeName=pelvis_capsule, center=<-4.9903 -0.2668 0>}]
 thigh_l [{
   shapeType=Capsule
   length=12.7454
   radius=9.0774
   shapeName=thigh_l_capsule
   center=<-10.9001 0.622 -0.2812>
   rotation=<rot -90 179.7016 -177.8014>
  }]
 thigh_r [{
   shapeType=Capsule
   length=13.0369
   radius=8.7859
   shapeName=thigh_l_capsule
   center=<10.9001 -0.4848 0.0896>
   rotation=<rot -90 -179.7016 -177.896>
  }]
 thigh_twist_01_l [{
   shapeType=Capsule
   length=10.0431
   radius=8.5845
   shapeName=thigh_twist_01_l_capsule
   center=<-2.4122 0 0>
   rotation=<rot -90 -6.7213 6.7212>
  }]
 thigh_twist_01_r [{shapeType=Capsule, length=10.0431, radius=8.5845, shapeName=thigh_twist_01_l_capsule, center=<2.4122 0 0>, rotation=<rot -90 -6.7213 6.7212>}]
}
constraints [
 {bone1Name=dyn_beltStripe01_l, bone2Name=MSTR_BeltStripes, swing1LimitDegrees=45, swing2LimitDegrees=45, twistLimitDegrees=5}
 {bone1Name=dyn_beltStripe02_l, bone2Name=dyn_beltStripe01_l, swing1LimitDegrees=30, swing2LimitDegrees=30, twistLimitDegrees=5}
 {bone1Name=dyn_beltStripe03_l, bone2Name=dyn_beltStripe02_l, swing1LimitDegrees=30, swing2LimitDegrees=30, twistLimitDegrees=5}
 {bone1Name=dyn_beltStripe01_r, bone2Name=MSTR_BeltStripes, swing1LimitDegrees=45, swing2LimitDegrees=45, twistLimitDegrees=5}
 {bone1Name=dyn_beltStripe02_r, bone2Name=dyn_beltStripe01_r, swing1LimitDegrees=30, swing2LimitDegrees=30, twistLimitDegrees=5}
]
