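# Physics asset for the SandalSite medium male player cosmetic body, set up for the physics of its two hoodie straps.
# Each strap is a chain of three capsule bodies hanging from a small sphere on its MSTR root bone.
# Constraints lock twist and limit swing, more tightly at the root than along the chain.
# Unconstrained capsules on the upper spine, clavicles and upper arms approximate the torso, apparently as colliders for the straps.
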PhysicsAsset /BRCosmetics/Characters/Player/Male/Medium/Bodies/M_MED_SandalSite/Meshes/M_MED_SandalSite_Physics
bodies {
 MSTR_hoodieStrape_l [{shapeType=Sphere, radius=0.505, shapeName=MSTR_hoodieStrape_l_sphere}]
 MSTR_hoodieStrape_r [{shapeType=Sphere, radius=0.505, shapeName=MSTR_hoodieStrape_l_sphere}]
 clavicle_l [{
   shapeType=Capsule
   length=6.0611
   radius=7.4777
   shapeName=clavicle_l_capsule
   center=<8.6516 3.9343 -3.1595>
   rotation=<rot -1.3656 79.1909 81.8929>
  }]
 clavicle_r [{
   shapeType=Capsule
   length=6.0611
   radius=7.4777
   shapeName=clavicle_r_capsule
   center=<-9.814 -4.1982 3.0362>
   rotation=<rot 1.4794 98.9895 -98.0872>
  }]
 dyn_hoodieStrape01_l [{
   shapeType=Capsule
   length=4.4638
   radius=1.1668
   shapeName=dyn_hoodieStrape01_l_capsule
   center=<2.6988 0 0>
   rotation=<rot -90 179.9999 180>
  }]
 dyn_hoodieStrape01_r [{
   shapeType=Capsule
   length=4.3346
   radius=1.2961
   shapeName=dyn_hoodieStrape01_l_capsule
   center=<2.6988 0 0>
   rotation=<rot -90 -3.3305 3.3304>
  }]
 dyn_hoodieStrape02_l [{
   shapeType=Capsule
   length=4.6541
   radius=1.4287
   shapeName=dyn_hoodieStrape02_l_capsule
   center=<3.1562 0 0>
   rotation=<rot 90 0.0001 0.0003>
  }]
 dyn_hoodieStrape02_r [{
   shapeType=Capsule
   length=6.0893
   radius=1.4469
   shapeName=dyn_hoodieStrape02_l_capsule
   center=<2.9688 0 0>
   rotation=<rot 90 180 -179.9998>
  }]
 dyn_hoodieStrape03_l [{
   shapeType=Capsule
   length=3.4129
   radius=1.1584
   shapeName=dyn_hoodieStrape03_l_capsule
   center=<2.4852 0 0.1781>
   rotation=<rot 86.3208 -0.0002 -0.0001>
  }]
 dyn_hoodieStrape03_r [{
   shapeType=Capsule
   length=3.2768
   radius=1.7044
   shapeName=dyn_hoodieStrape03_l_capsule
   center=<3.0693 0 0>
   rotation=<rot 83.3268 -0.0001 0>
  }]
 spine_04 [{shapeType=Capsule, length=14.1611, radius=10.8389, shapeName=spine_04_capsule, center=<1.6498 3.6317 0>}]
 spine_05 [{shapeType=Capsule, length=17.4941, radius=13.7116, shapeName=spine_05_capsule, center=<-4.549 -0.2777 -0.0088>}]
 upperarm_l [{
   shapeType=Capsule
   length=19.9065
   radius=7.7142
   shapeName=upperarm_l_capsule
   center=<8.5122 -0.6772 1.4723>
   rotation=<rot -80 179.9998 -179.9998>
  }]
 upperarm_r [{
   shapeType=Capsule
   length=19.9065
   radius=7.7142
   shapeName=upperarm_r_capsule
   center=<-8.5122 0.6772 -1.4723>
   rotation=<rot -80 179.9998 0.0002>
  }]
}
constraints [
 {
  bone1Name=dyn_hoodieStrape02_l
  bone2Name=dyn_hoodieStrape01_l
  swing1LimitDegrees=45
  swing2LimitDegrees=45
  twistLimitDegrees=45
  twistMotion=Locked
 }
 {
  bone1Name=dyn_hoodieStrape03_l
  bone2Name=dyn_hoodieStrape02_l
  swing1LimitDegrees=45
  swing2LimitDegrees=45
  twistLimitDegrees=45
  twistMotion=Locked
 }
 {
  bone1Name=dyn_hoodieStrape01_l
  bone2Name=MSTR_hoodieStrape_l
  swing1LimitDegrees=30
  swing2LimitDegrees=10
  twistLimitDegrees=5
  twistMotion=Locked
 }
 {
  bone1Name=dyn_hoodieStrape01_r
  bone2Name=MSTR_hoodieStrape_r
  swing1LimitDegrees=30
  swing2LimitDegrees=10
  twistLimitDegrees=5
  twistMotion=Locked
 }
 {
  bone1Name=dyn_hoodieStrape02_r
  bone2Name=dyn_hoodieStrape01_r
  swing1LimitDegrees=45
  swing2LimitDegrees=45
  twistLimitDegrees=45
  twistMotion=Locked
 }
 {
  bone1Name=dyn_hoodieStrape03_r
  bone2Name=dyn_hoodieStrape02_r
  swing1LimitDegrees=45
  swing2LimitDegrees=45
  twistLimitDegrees=45
  twistMotion=Locked
 }
]
