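# Animation Blueprint for the M_MED_SandalSite male body cosmetic, adding secondary motion to hood, pockets, belts and foot straps.
# Parent mesh pose passes through a Control Rig, then weighted bone constraints and a spring-and-collision AnimDynamics chain, then two RigidBody simulations.
# Shirt dynamics and hoodie-corner constraint chains are not connected to the output.
# The event graph tracks velocity, crouch, LOD, gravity override and cape backbling state.
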AnimBlueprint /BRCosmetics/Characters/Player/Male/Medium/Bodies/M_MED_SandalSite/Meshes/M_MED_SandalSite_AnimBP
target_skeleton=&/M_MED_SandalSite_Skeleton
graphs [
 AnimGraph {nodes [
   n1 = Anim Dynamics - Bone: dyn_shirt01(Alpha=1, ExternalForce="0,0,0") {
    angularDampingOverride=0.9
    bEnableWind=true
    bOverrideAngularDamping=true
    bUseGravityOverride=true
    simulationSpace=RootRelative
    windScale=10
    alphaBoolBlend {blendOption=Linear}
    boundBone {boneName=dyn_shirt01}
    gravityOverride {z=-980}
    physicsBodyDefinitions [{
      boundBone {boneName=dyn_shirt01}
      boxExtents=<15 15 15>
      constraintSetup {angularLimitsMax=<0 0 5>, angularLimitsMin=<0 -20 -5>, angularTarget=<1 0 0>}
      localJointOffset=<0.15 0 0>
     }]
   }
   Anim Dynamics - Bone: dyn_shirt02(ComponentPose=@n1, Alpha=1, ExternalForce="0,0,0") {
    angularDampingOverride=0.9
    bEnableWind=true
    bOverrideAngularDamping=true
    bUseGravityOverride=true
    simulationSpace=RootRelative
    windScale=10
    alphaBoolBlend {blendOption=Linear}
    boundBone {boneName=dyn_shirt02}
    gravityOverride {z=-980}
    physicsBodyDefinitions [{
      boundBone {boneName=dyn_shirt02}
      boxExtents=<15 15 15>
      constraintSetup {angularLimitsMax=<0 0 5>, angularLimitsMin=<0 -20 -5>, angularTarget=<1 0 0>}
      localJointOffset=<0.15 0 0>
     }]
   }
   n2 = "Constraint - def_hoodieCorner_r (PPPPPPPP)"!(
    ComponentPose
    Alpha=1
    ConstraintWeights_0=0.0027
    ConstraintWeights_1:float=0.4564
    ConstraintWeights_2:float=0.0252
    ConstraintWeights_3:float=0.3224
    ConstraintWeights_4:float=0.0011
    ConstraintWeights_5:float=0.1294
    ConstraintWeights_6:float=0.0536
    ConstraintWeights_7:float=0.0093
   ) -> (Pose) {
    alphaBoolBlend {blendOption=Linear}
    boneToModify {boneName=def_hoodieCorner_r}
    constraintSetup [
     {targetBone {boneName=spine_03}}
     {targetBone {boneName=spine_05}}
     {targetBone {boneName=neck_01}}
     {targetBone {boneName=clavicle_r}}
     {targetBone {boneName=upperarm_twist_01_r}}
     {targetBone {boneName=deform_upperarm_up_r}}
     {targetBone {boneName=neck_02}}
     {targetBone {boneName=deform_upperarm_fr_r}}
    ]
    constraintWeights [1,1,1,1,1,1,1,1]
   }
   "Constraint - def_hoodieCorner_l (PPPPPP)"!(
    ComponentPose=@n2
    Alpha=1
    ConstraintWeights_0=0.4952
    ConstraintWeights_1:float=0.0286
    ConstraintWeights_2:float=0.036
    ConstraintWeights_3:float=0.0762
    ConstraintWeights_4:float=0.2952
    ConstraintWeights_5:float=0.0688
   ) -> (Pose) {
    alphaBoolBlend {blendOption=Linear}
    boneToModify {boneName=def_hoodieCorner_l}
    constraintSetup [
     {targetBone {boneName=spine_05}}
     {targetBone {boneName=neck_01}}
     {targetBone {boneName=upperarm_twist_01_l}}
     {targetBone {boneName=deform_upperarm_up_l}}
     {targetBone {boneName=clavicle_l}}
     {targetBone {boneName=neck_02}}
    ]
    constraintWeights [1,1,1,1,1,1]
   }
   n3 = RigidBody!(ComponentPose=Local To Component(LocalPose=Use cached pose 'AnimDynamics'()), Alpha, OverrideWorldGravity:Vector="0,0,0", ExternalForce) -> (Pose) {
    bClampLinearTranslationLimitToRefPose=true
    bOverrideWorldGravity=true
    overridePhysicsAsset=&/M_MED_SandalSiteBelt_Physics
    simulationSpace=BaseBoneSpace
    alphaBoolBlend {blendOption=Linear}
    baseBoneRef {boneName=pelvis}
    simSpaceSettings {maxAngularAcceleration=300, maxAngularVelocity=30, maxLinearAcceleration=10, maxLinearVelocity=100, velocityScaleZ=0.7, worldAlpha=0.15}
   }
   n4 = RigidBody(ComponentPose=@n3, OverrideWorldGravity="0,0,0") {
    bOverrideWorldGravity=true
    overridePhysicsAsset=&/M_MED_SandalSite_Physics
    simulationSpace=BaseBoneSpace
    alphaBoolBlend {blendOption=Linear}
    baseBoneRef {boneName=spine_05}
    simSpaceSettings {maxAngularAcceleration=300, maxAngularVelocity=30, maxLinearAcceleration=10, maxLinearVelocity=100, velocityScaleZ=0.7, worldAlpha=0.15}
   }
   Output Pose(Result=Component To Local(ComponentPose=@n4)) {layerGroup=None}
   n5 = Control Rig!(
    Source=Copy Pose From Mesh() {bUseAttachedParent=true}
    Alpha
    isMoving2D:bool=Get IsVelocityZero() -> (IsVelocityZero:bool)
    isCrouching:bool=Get IsCrouching() -> (IsCrouching:bool)
   ) -> (Pose) {
    alphaBoolBlend {blendOption=Linear}
    controlRigAssetReference {blueprintRigClass=&/M_MED_SandalSite_CtrlRig.M_MED_SandalSite_CtrlRig_C}
   }
   Save cached pose 'PreSim'(Pose=@n5) {cacheName=PreSim}
   n6 = "Constraint - MSTR_hoodieStrape_l (PPPPPPPP)"!(
    ComponentPose=Local To Component(LocalPose=Use cached pose 'PreSim'())
    Alpha=1
    ConstraintWeights_0=0.0186
    ConstraintWeights_1:float=0.0641
    ConstraintWeights_2:float=0.6344
    ConstraintWeights_3:float=0.0358
    ConstraintWeights_4:float=0.1953
    ConstraintWeights_5:float=0.005
    ConstraintWeights_6:float=0.0202
    ConstraintWeights_7:float=0.0266
   ) -> (Pose) {
    alphaBoolBlend {blendOption=Linear}
    boneToModify {boneName=MSTR_hoodieStrape_l}
    constraintSetup [
     {targetBone {boneName=spine_03}}
     {targetBone {boneName=spine_04}}
     {targetBone {boneName=spine_05}}
     {targetBone {boneName=neck_01}}
     {targetBone {boneName=def_hoodieCorner_l}}
     {targetBone {boneName=dyn_hoodieStrape01_l}}
     {targetBone {boneName=deform_pec_l}}
     {targetBone {boneName=clavicle_l}}
    ]
    constraintWeights [1,1,1,1,1,1,1,1]
   }
   n7 = "Constraint - MSTR_hoodieStrape_r (PPPPPPPP)"!(
    ComponentPose=@n6
    Alpha=1
    ConstraintWeights_0=0.0027
    ConstraintWeights_1:float=0.0169
    ConstraintWeights_2:float=0.0602
    ConstraintWeights_3:float=0.511
    ConstraintWeights_4:float=0.0312
    ConstraintWeights_5:float=0.0698
    ConstraintWeights_6:float=0.3037
    ConstraintWeights_7:float=0.0046
   ) -> (Pose) {
    alphaBoolBlend {blendOption=Linear}
    boneToModify {boneName=MSTR_hoodieStrape_r}
    constraintSetup [
     {targetBone {boneName=spine_02}}
     {targetBone {boneName=spine_03}}
     {targetBone {boneName=spine_04}}
     {targetBone {boneName=spine_05}}
     {targetBone {boneName=neck_01}}
     {targetBone {boneName=clavicle_r}}
     {targetBone {boneName=def_hoodieCorner_r}}
     {targetBone {boneName=dyn_hoodieStrape01_r}}
    ]
    constraintWeights [1,1,1,1,1,1,1,1]
   }
   n8 = "Constraint - dyn_pelvisCorrColl (PPPPPPPP)"!(
    ComponentPose=@n7
    Alpha=1
    ConstraintWeights_0=0.2745
    ConstraintWeights_1:float=0.0075
    ConstraintWeights_2:float=0.2739
    ConstraintWeights_3:float=0.0717
    ConstraintWeights_4:float=0.2379
    ConstraintWeights_5:float=0.088
    ConstraintWeights_6:float=0.0234
    ConstraintWeights_7:float=0.0231
   ) -> (Pose) {
    alphaBoolBlend {blendOption=Linear}
    boneToModify {boneName=dyn_pelvisCorrColl}
    constraintSetup [
     {targetBone {boneName=pelvis}}
     {targetBone {boneName=thigh_l}}
     {targetBone {boneName=thigh_twist_01_l}}
     {targetBone {boneName=deform_groin_l}}
     {targetBone {boneName=thigh_twist_01_r}}
     {targetBone {boneName=deform_groin_r}}
     {targetBone {boneName=deform_glute_l}}
     {targetBone {boneName=deform_glute_r}}
    ]
    constraintWeights [1,1,1,1,1,1,1,1]
   }
   n9 = "Constraint - MSTR_chetBelt (PPPPPPP)"!(
    ComponentPose=@n8
    Alpha=1
    ConstraintWeights_0=0.3536
    ConstraintWeights_1:float=0.1675
    ConstraintWeights_2:float=0.2682
    ConstraintWeights_3:float=0.1741
    ConstraintWeights_4:float=0.0359
    ConstraintWeights_5:float=0.0006
    ConstraintWeights_6:float=0.0002
   ) -> (Pose) {
    alphaBoolBlend {blendOption=Linear}
    boneToModify {boneName=MSTR_chetBelt}
    constraintSetup [
     {targetBone {boneName=pelvis}}
     {targetBone {boneName=spine_01}}
     {targetBone {boneName=spine_02}}
     {targetBone {boneName=spine_03}}
     {targetBone {boneName=spine_04}}
     {targetBone {boneName=deform_upperarm_bk_l}}
     {targetBone {boneName=deform_pec_l}}
    ]
    constraintWeights [1,1,1,1,1,1,1]
   }
   n10 = "Constraint - def_chestBeltBackPin (PPPPPP)"!(
    ComponentPose=@n9
    Alpha=1
    ConstraintWeights_0=0.1544
    ConstraintWeights_1:float=0.147
    ConstraintWeights_2:float=0.2908
    ConstraintWeights_3:float=0.3071
    ConstraintWeights_4:float=0.0958
    ConstraintWeights_5:float=0.0048
   ) -> (Pose) {
    alphaBoolBlend {blendOption=Linear}
    boneToModify {boneName=def_chestBeltBackPin}
    constraintSetup [
     {targetBone {boneName=pelvis}}
     {targetBone {boneName=spine_01}}
     {targetBone {boneName=spine_02}}
     {targetBone {boneName=spine_03}}
     {targetBone {boneName=spine_04}}
     {targetBone {boneName=deform_upperarm_bk_l}}
    ]
    constraintWeights [1,1,1,1,1,1]
   }
   n11 = Anim Dynamics - Bone: dyn_FootStrapeIn01_l(ComponentPose=@n10, Alpha=1, ExternalForce="0,0,0") {
    angularDampingOverride=0.999
    angularSpringConstant=300
    bAngularSpring=true
    bOverrideAngularDamping=true
    bUseGravityOverride=true
    simulationSpace=RootRelative
    alphaBoolBlend {blendOption=Linear}
    boundBone {boneName=dyn_FootStrapeIn01_l}
    componentAppliedLinearAccClamp=<1000 1000 1000>
    gravityOverride {z=-980}
    physicsBodyDefinitions [{
      boundBone {boneName=dyn_FootStrapeIn01_l}
      boxExtents=<20 20 20>
      constraintSetup {angularLimitsMax=<0 10 5>, angularLimitsMin=<0 -30 0>, angularTarget=<1 0 0>}
      localJointOffset=<4 0 0>
     }]
   }
   n12 = Anim Dynamics - Bone: dyn_FootStrapeIn02_l(ComponentPose=@n11, Alpha=1, ExternalForce="0,0,0") {
    angularDampingOverride=0.999
    angularSpringConstant=300
    bAngularSpring=true
    bOverrideAngularDamping=true
    bUseGravityOverride=true
    bUsePlanarLimit=false
    bUseSphericalLimits=true
    simulationSpace=BoneRelative
    alphaBoolBlend {blendOption=Linear}
    boundBone {boneName=dyn_FootStrapeIn02_l}
    componentAppliedLinearAccClamp=<1000 1000 1000>
    gravityOverride {z=-980}
    physicsBodyDefinitions [{
      collisionType=CustomSphere
      sphereCollisionRadius=1.5
      boundBone {boneName=dyn_FootStrapeIn02_l}
      boxExtents=<20 20 20>
      constraintSetup {angularLimitsMax=<0 10 5>, angularLimitsMin=<0 -20 0>, angularTarget=<1 0 0>}
      localJointOffset=<4 0 0>
     }]
    planarLimits [{drivingBone {boneName=foot_l}, planeTransform {location=<5.5 0 0>, rotation=<rot 0 -45 80>}}]
    relativeSpaceBone {boneName=dyn_FootStrapeIn01_l}
    sphericalLimits [
     {limitRadius=5.3, drivingBone {boneName=foot_l}, sphereLocalOffset=<0 0 0>}
     {limitRadius=3.5, drivingBone {boneName=foot_l}, sphereLocalOffset=<-3 6 1>}
    ]
   }
   n13 = Anim Dynamics - Bone: dyn_FootStrapeOut01_l(ComponentPose=@n12, Alpha=1, ExternalForce="0,0,0") {
    angularDampingOverride=0.999
    angularSpringConstant=300
    bAngularSpring=true
    bOverrideAngularDamping=true
    bUseGravityOverride=true
    simulationSpace=RootRelative
    alphaBoolBlend {blendOption=Linear}
    boundBone {boneName=dyn_FootStrapeOut01_l}
    componentAppliedLinearAccClamp=<1000 1000 1000>
    gravityOverride {z=-980}
    physicsBodyDefinitions [{
      boundBone {boneName=dyn_FootStrapeOut01_l}
      boxExtents=<20 20 20>
      constraintSetup {angularLimitsMax=<0 10 0>, angularLimitsMin=<0 -30 -5>, angularTarget=<1 0 0>}
      localJointOffset=<4 0 0>
     }]
   }
   n14 = Anim Dynamics - Bone: dyn_FootStrapeOut02_l(ComponentPose=@n13, Alpha=1, ExternalForce="0,0,0") {
    angularDampingOverride=0.999
    angularSpringConstant=300
    bAngularSpring=true
    bOverrideAngularDamping=true
    bUseGravityOverride=true
    bUsePlanarLimit=false
    bUseSphericalLimits=true
    simulationSpace=BoneRelative
    alphaBoolBlend {blendOption=Linear}
    boundBone {boneName=dyn_FootStrapeOut02_l}
    componentAppliedLinearAccClamp=<1000 1000 1000>
    gravityOverride {z=-980}
    physicsBodyDefinitions [{
      collisionType=CustomSphere
      sphereCollisionRadius=2
      boundBone {boneName=dyn_FootStrapeOut02_l}
      boxExtents=<10 10 10>
      constraintSetup {angularLimitsMax=<0 10 0>, angularLimitsMin=<0 -20 -5>, angularTarget=<1 0 0>}
      localJointOffset=<4 0 0>
     }]
    planarLimits [{drivingBone {boneName=foot_l}, planeTransform {location=<5.5 0 0>, rotation=<rot 0 -45 79.9999>}}]
    relativeSpaceBone {boneName=dyn_FootStrapeOut01_l}
    sphericalLimits [
     {limitRadius=5.3, drivingBone {boneName=foot_l}, sphereLocalOffset=<0 0 0>}
     {limitRadius=3.5, drivingBone {boneName=foot_l}, sphereLocalOffset=<-3 6 -1.5>}
    ]
   }
   n15 = Anim Dynamics - Bone: dyn_FootStrapeIn01_r(ComponentPose=@n14, Alpha=1, ExternalForce="0,0,0") {
    angularDampingOverride=0.999
    angularSpringConstant=300
    bAngularSpring=true
    bOverrideAngularDamping=true
    bUseGravityOverride=true
    simulationSpace=RootRelative
    alphaBoolBlend {blendOption=Linear}
    boundBone {boneName=dyn_FootStrapeIn01_r}
    componentAppliedLinearAccClamp=<1000 1000 1000>
    gravityOverride {z=-980}
    physicsBodyDefinitions [{
      boundBone {boneName=dyn_FootStrapeIn01_r}
      boxExtents=<20 20 20>
      constraintSetup {angularLimitsMax=<0 10 0>, angularLimitsMin=<0 -30 -5>, angularTarget=<1 0 0>}
      localJointOffset=<4 0 0>
     }]
   }
   n16 = Anim Dynamics - Bone: dyn_FootStrapeIn02_r(ComponentPose=@n15, Alpha=1, ExternalForce="0,0,0") {
    angularDampingOverride=0.999
    angularSpringConstant=300
    bAngularSpring=true
    bOverrideAngularDamping=true
    bUseGravityOverride=true
    bUsePlanarLimit=false
    bUseSphericalLimits=true
    simulationSpace=BoneRelative
    alphaBoolBlend {blendOption=Linear}
    boundBone {boneName=dyn_FootStrapeIn02_r}
    componentAppliedLinearAccClamp=<1000 1000 1000>
    gravityOverride {z=-980}
    physicsBodyDefinitions [{
      collisionType=CustomSphere
      sphereCollisionRadius=1.5
      boundBone {boneName=dyn_FootStrapeIn02_r}
      boxExtents=<20 20 20>
      constraintSetup {angularLimitsMax=<0 10 0>, angularLimitsMin=<0 -20 -5>, angularTarget=<1 0 0>}
      localJointOffset=<4 0 0>
     }]
    planarLimits [{drivingBone {boneName=foot_l}, planeTransform {location=<5.5 0 0>, rotation=<rot 0 -45 79.9999>}}]
    relativeSpaceBone {boneName=dyn_FootStrapeIn01_r}
    sphericalLimits [
     {limitRadius=5.3, drivingBone {boneName=foot_r}, sphereLocalOffset=<0 0 0>}
     {limitRadius=3.5, drivingBone {boneName=foot_r}, sphereLocalOffset=<3 -6 -1>}
    ]
   }
   n17 = Anim Dynamics - Bone: dyn_FootStrapeOut01_r(ComponentPose=@n16, Alpha=1, ExternalForce="0,0,0") {
    angularDampingOverride=0.999
    angularSpringConstant=300
    bAngularSpring=true
    bOverrideAngularDamping=true
    bUseGravityOverride=true
    simulationSpace=RootRelative
    alphaBoolBlend {blendOption=Linear}
    boundBone {boneName=dyn_FootStrapeOut01_r}
    componentAppliedLinearAccClamp=<1000 1000 1000>
    gravityOverride {z=-980}
    physicsBodyDefinitions [{
      boundBone {boneName=dyn_FootStrapeOut01_r}
      boxExtents=<20 20 20>
      constraintSetup {angularLimitsMax=<0 10 5>, angularLimitsMin=<0 -30 0>, angularTarget=<1 0 0>}
      localJointOffset=<4 0 0>
     }]
   }
   n18 = Anim Dynamics - Bone: dyn_FootStrapeOut02_r(ComponentPose=@n17, Alpha=1, ExternalForce="0,0,0") {
    angularDampingOverride=0.999
    angularSpringConstant=300
    bAngularSpring=true
    bOverrideAngularDamping=true
    bUseGravityOverride=true
    bUsePlanarLimit=false
    bUseSphericalLimits=true
    simulationSpace=BoneRelative
    alphaBoolBlend {blendOption=Linear}
    boundBone {boneName=dyn_FootStrapeOut02_r}
    componentAppliedLinearAccClamp=<1000 1000 1000>
    gravityOverride {z=-980}
    physicsBodyDefinitions [{
      collisionType=CustomSphere
      sphereCollisionRadius=2
      boundBone {boneName=dyn_FootStrapeOut02_r}
      boxExtents=<10 10 10>
      constraintSetup {angularLimitsMax=<0 10 5>, angularLimitsMin=<0 -20 0>, angularTarget=<1 0 0>}
      localJointOffset=<4 0 0>
     }]
    planarLimits [{drivingBone {boneName=foot_l}, planeTransform {location=<5.5 0 0>, rotation=<rot 0 -45 79.9999>}}]
    relativeSpaceBone {boneName=dyn_FootStrapeOut01_r}
    sphericalLimits [
     {limitRadius=5.3, drivingBone {boneName=foot_r}, sphereLocalOffset=<0 -1 0>}
     {limitRadius=3.5, drivingBone {boneName=foot_r}, sphereLocalOffset=<3 -7 1.5>}
    ]
   }
   n19 = Anim Dynamics - Bone: dyn_hoodieTip!(ComponentPose=@n18, bAlphaBoolEnabled:bool=True, GravityOverride:Vector="0,0,-980", ExternalForce="0,0,0") -> (Pose) {
    alphaInputType=Bool
    angularDampingOverride=0.99
    angularSpringConstant=500
    bAngularSpring=true
    bOverrideAngularDamping=true
    bUseGravityOverride=true
    simulationSpace=RootRelative
    alphaBoolBlend {blendOption=Linear}
    boundBone {boneName=dyn_hoodieTip}
    componentAppliedLinearAccClamp=<3000 3000 3000>
    gravityOverride {z=-980}
    physicsBodyDefinitions [{
      boundBone {boneName=dyn_hoodieTip}
      boxExtents=<15 15 15>
      constraintSetup {angularLimitsMax=<0 0 2>, angularLimitsMin=<0 0 -20>, angularTarget=<0 0 0>}
      localJointOffset=<0.3 0 0>
     }]
    relativeSpaceBone {boneName=root}
   }
   n20 = Anim Dynamics - Bone: dyn_pocket_l!(ComponentPose=@n19, Alpha=1, GravityOverride:Vector="0,0,-980", ExternalForce="0,0,0") -> (Pose) {
    bOverrideLinearDamping=true
    bUseGravityOverride=true
    linearDampingOverride=0.999
    simulationSpace=RootRelative
    alphaBoolBlend {blendOption=Linear}
    boundBone {boneName=dyn_pocket_l}
    physicsBodyDefinitions [{
      boundBone {boneName=dyn_pocket_l}
      boxExtents=<20 20 20>
      constraintSetup {angularLimitsMax=<0 0 0>, angularLimitsMin=<0 0 0>, angularTarget=<0 0 0>, linearAxesMax=<0.2 0.2 0.1>, linearAxesMin=<-0.2 -0.2 -0.1>}
      localJointOffset=<0.1 0 0>
     }]
   }
   n21 = Anim Dynamics - Bone: dyn_pocket_r!(ComponentPose=@n20, Alpha=1, GravityOverride:Vector="0,0,-980", ExternalForce="0,0,0") -> (Pose) {
    bOverrideLinearDamping=true
    bUseGravityOverride=true
    linearDampingOverride=0.999
    simulationSpace=RootRelative
    alphaBoolBlend {blendOption=Linear}
    boundBone {boneName=dyn_pocket_r}
    physicsBodyDefinitions [{
      boundBone {boneName=dyn_pocket_r}
      boxExtents=<20 20 20>
      constraintSetup {angularLimitsMax=<0 0 0>, angularLimitsMin=<0 0 0>, angularTarget=<0 0 0>, linearAxesMax=<0.2 0.2 0.1>, linearAxesMin=<-0.2 -0.2 -0.1>}
      localJointOffset=<0.1 0 0>
     }]
   }
   n22 = Anim Dynamics - Bone: dyn_chetBelt(ComponentPose=@n21, Alpha=1, ExternalForce="0,0,0") {
    angularDampingOverride=0.9
    bEnableWind=true
    bLinearSpring=true
    bOverrideLinearDamping=true
    bUseGravityOverride=true
    linearDampingOverride=0.9999
    linearSpringConstant=50
    simulationSpace=RootRelative
    windScale=10
    alphaBoolBlend {blendOption=Linear}
    boundBone {boneName=dyn_chetBelt}
    componentAppliedLinearAccClamp=<10000 10000>
    gravityOverride {z=-980}
    physicsBodyDefinitions [{
      boundBone {boneName=dyn_chetBelt}
      boxExtents=<30 30 30>
      constraintSetup {angularLimitsMax=<0 0 0>, angularLimitsMin=<0 0 0>, angularTarget=<1 0 0>, linearAxesMax=<0 0 2>, linearAxesMin=<0 0 -0.1>}
      localJointOffset=<0 0 -0.15>
     }]
   }
   n23 = "Constraint - dyn_beltStripe01_l (RR)"!(ComponentPose=@n22, bAlphaBoolEnabled:bool=True, ConstraintWeights_0=0.8, ConstraintWeights_1:float=0.2) -> (Pose) {
    alphaInputType=Bool
    alphaBoolBlend {blendOption=Linear}
    boneToModify {boneName=dyn_beltStripe01_l}
    constraintSetup [{transformType=Rotation, targetBone {boneName=thigh_l}}, {transformType=Rotation, targetBone {boneName=pelvis}}]
    constraintWeights [1,1]
   }
   n24 = "Constraint - dyn_beltStripe01_r (RR)"!(ComponentPose=@n23, bAlphaBoolEnabled:bool=True, ConstraintWeights_0=0.8, ConstraintWeights_1:float=0.2) -> (Pose) {
    alphaInputType=Bool
    alphaBoolBlend {blendOption=Linear}
    boneToModify {boneName=dyn_beltStripe01_r}
    constraintSetup [{transformType=Rotation, targetBone {boneName=thigh_r}}, {transformType=Rotation, targetBone {boneName=pelvis}}]
    constraintWeights [1,1]
   }
   n25 = Anim Dynamics - Bone: dyn_beltStripe01_r(ComponentPose=@n24, Alpha=1, ExternalForce="0,0,0") {
    angularDampingOverride=0.99
    angularSpringConstant=75
    bAngularSpring=true
    bOverrideAngularDamping=true
    bUseGravityOverride=true
    bUseSphericalLimits=true
    simulationSpace=RootRelative
    alphaBoolBlend {blendOption=Linear}
    boundBone {boneName=dyn_beltStripe01_r}
    componentAppliedLinearAccClamp=<5000 5000 5000>
    gravityOverride {z=-980}
    physicsBodyDefinitions [{
      collisionType=CustomSphere
      sphereCollisionRadius=2
      boundBone {boneName=dyn_beltStripe01_r}
      boxExtents=<10 5 10>
      constraintSetup {angularLimitsMax=<0 0 10>, angularLimitsMin=<0 0 -10>, angularTarget=<0 0 0>}
      localJointOffset=<-0.2 0 0>
     }]
    relativeSpaceBone {boneName=dyn_beltStripe01_r}
    sphericalLimits [
     {limitRadius=9, drivingBone {boneName=pelvis}, sphereLocalOffset=<-2 1.5 0>}
     {limitRadius=10, drivingBone {boneName=thigh_r}, sphereLocalOffset=<3 0 0>}
     {limitRadius=5, drivingBone {boneName=dyn_pelvisCorrColl}, sphereLocalOffset=<0 0 0>}
    ]
   }
   n26 = Anim Dynamics - Bone: dyn_beltStripe02_r(ComponentPose=@n25, Alpha=1, ExternalForce="0,0,0") {
    angularDampingOverride=0.99
    angularSpringConstant=75
    bAngularSpring=true
    bOverrideAngularDamping=true
    bUseGravityOverride=true
    bUseSphericalLimits=true
    simulationSpace=RootRelative
    alphaBoolBlend {blendOption=Linear}
    boundBone {boneName=dyn_beltStripe02_r}
    componentAppliedLinearAccClamp=<5000 5000 5000>
    gravityOverride {z=-980}
    physicsBodyDefinitions [{
      collisionType=CustomSphere
      sphereCollisionRadius=2
      boundBone {boneName=dyn_beltStripe02_r}
      boxExtents=<10 5 10>
      constraintSetup {angularLimitsMax=<0 0 5>, angularLimitsMin=<0 0 -30>, angularTarget=<0 0 0>}
      localJointOffset=<-0.2 0 0>
     }]
    relativeSpaceBone {boneName=dyn_beltStripe01_r}
    sphericalLimits [
     {limitRadius=8.5, drivingBone {boneName=pelvis}, sphereLocalOffset=<-2 1.5 4>}
     {limitRadius=10, drivingBone {boneName=thigh_r}, sphereLocalOffset=<15 0 -1>}
     {limitRadius=5, drivingBone {boneName=dyn_pelvisCorrColl}, sphereLocalOffset=<0 0 0>}
    ]
   }
   n27 = Anim Dynamics - Bone: dyn_beltStripe01_l(ComponentPose=@n26, Alpha=1, ExternalForce="0,0,0") {
    angularDampingOverride=0.99
    angularSpringConstant=75
    bAngularSpring=true
    bOverrideAngularDamping=true
    bUseGravityOverride=true
    bUseSphericalLimits=true
    simulationSpace=RootRelative
    alphaBoolBlend {blendOption=Linear}
    boundBone {boneName=dyn_beltStripe01_l}
    componentAppliedLinearAccClamp=<5000 5000 5000>
    gravityOverride {z=-980}
    physicsBodyDefinitions [{
      collisionType=CustomSphere
      sphereCollisionRadius=2
      boundBone {boneName=dyn_beltStripe01_l}
      boxExtents=<10 5 10>
      constraintSetup {angularLimitsMax=<0 0 10>, angularLimitsMin=<0 0 -10>, angularTarget=<0 0 0>}
      localJointOffset=<0.2 0 0>
     }]
    relativeSpaceBone {boneName=dyn_beltStripe01_r}
    sphericalLimits [
     {limitRadius=9, drivingBone {boneName=pelvis}, sphereLocalOffset=<-2 1.5 0>}
     {limitRadius=10, drivingBone {boneName=thigh_l}, sphereLocalOffset=<-3 0 0>}
     {limitRadius=5, drivingBone {boneName=dyn_pelvisCorrColl}, sphereLocalOffset=<0 0 0>}
    ]
   }
   n28 = Anim Dynamics - Bone: dyn_beltStripe02_l(ComponentPose=@n27, Alpha=1, ExternalForce="0,0,0") {
    angularDampingOverride=0.99
    angularSpringConstant=75
    bAngularSpring=true
    bOverrideAngularDamping=true
    bUseGravityOverride=true
    bUseSphericalLimits=true
    simulationSpace=RootRelative
    alphaBoolBlend {blendOption=Linear}
    boundBone {boneName=dyn_beltStripe02_l}
    componentAppliedLinearAccClamp=<5000 5000 5000>
    gravityOverride {z=-980}
    physicsBodyDefinitions [{
      collisionType=CustomSphere
      sphereCollisionRadius=2
      boundBone {boneName=dyn_beltStripe02_l}
      boxExtents=<10 5 10>
      constraintSetup {angularLimitsMax=<0 0 5>, angularLimitsMin=<0 0 -30>, angularTarget=<0 0 0>}
      localJointOffset=<0.2 0 0>
     }]
    relativeSpaceBone {boneName=dyn_beltStripe01_r}
    sphericalLimits [
     {limitRadius=8.5, drivingBone {boneName=pelvis}, sphereLocalOffset=<-4 1.5 -4>}
     {limitRadius=10, drivingBone {boneName=thigh_l}, sphereLocalOffset=<-15 0 1>}
     {limitRadius=5, drivingBone {boneName=dyn_pelvisCorrColl}, sphereLocalOffset=<0 0 0>}
    ]
   }
   n29 = Anim Dynamics - Bone: dyn_beltStripe03_l(ComponentPose=@n28, Alpha=1, ExternalForce="0,0,0") {
    angularDampingOverride=0.99
    angularSpringConstant=75
    bAngularSpring=true
    bOverrideAngularDamping=true
    bUseGravityOverride=true
    bUseSphericalLimits=true
    simulationSpace=RootRelative
    alphaBoolBlend {blendOption=Linear}
    boundBone {boneName=dyn_beltStripe03_l}
    componentAppliedLinearAccClamp=<5000 5000 5000>
    gravityOverride {z=-980}
    physicsBodyDefinitions [{
      collisionType=CustomSphere
      sphereCollisionRadius=2
      boundBone {boneName=dyn_beltStripe03_l}
      boxExtents=<10 5 10>
      constraintSetup {angularLimitsMax=<0 0 5>, angularLimitsMin=<0 0 -30>, angularTarget=<0 0 0>}
      localJointOffset=<0.2 0 0>
     }]
    relativeSpaceBone {boneName=dyn_beltStripe01_r}
    sphericalLimits [
     {limitRadius=10, drivingBone {boneName=thigh_l}, sphereLocalOffset=<-15 0 1>}
     {limitRadius=5, drivingBone {boneName=dyn_pelvisCorrColl}, sphereLocalOffset=<0 0 0>}
    ]
   }
   n30 = "Constraint - dyn_hoodieStrape02_r (PPPPPPPP)"!(
    ComponentPose=@n29
    bAlphaBoolEnabled:bool=True
    ConstraintWeights_0=0.0442
    ConstraintWeights_1:float=0.122
    ConstraintWeights_2:float=0.3215
    ConstraintWeights_3:float=0.3446
    ConstraintWeights_4:float=0.0001
    ConstraintWeights_5:float=0.1633
    ConstraintWeights_6:float=0.0003
    ConstraintWeights_7:float=0.004
   ) -> (Pose) {
    alphaInputType=Bool
    alphaBoolBlend {blendOption=Linear}
    boneToModify {boneName=dyn_hoodieStrape02_r}
    constraintSetup [
     {targetBone {boneName=spine_02}}
     {targetBone {boneName=spine_03}}
     {targetBone {boneName=spine_04}}
     {targetBone {boneName=spine_05}}
     {targetBone {boneName=neck_01}}
     {targetBone {boneName=clavicle_r}}
     {targetBone {boneName=deform_upperarm_up_r}}
     {targetBone {boneName=deform_upperarm_fr_r}}
    ]
    constraintWeights [1,1,1,1,1,1,1,1]
   }
   n31 = "Constraint - dyn_hoodieStrape03_r (PPPPPPPP)"!(
    ComponentPose=@n30
    bAlphaBoolEnabled:bool=True
    ConstraintWeights_0=0
    ConstraintWeights_1:float=0.0708
    ConstraintWeights_2:float=0.1755
    ConstraintWeights_3:float=0.3838
    ConstraintWeights_4:float=0.2793
    ConstraintWeights_5:float=0.0786
    ConstraintWeights_6:float=0.0001
    ConstraintWeights_7:float=0.0119
   ) -> (Pose) {
    alphaInputType=Bool
    alphaBoolBlend {blendOption=Linear}
    boneToModify {boneName=dyn_hoodieStrape03_r}
    constraintSetup [
     {targetBone {boneName=spine_01}}
     {targetBone {boneName=spine_02}}
     {targetBone {boneName=spine_03}}
     {targetBone {boneName=spine_04}}
     {targetBone {boneName=spine_05}}
     {targetBone {boneName=clavicle_r}}
     {targetBone {boneName=deform_pec_l}}
     {targetBone {boneName=deform_upperarm_fr_r}}
    ]
    constraintWeights [1,1,1,1,1,1,1,1]
   }
   n32 = "Constraint - dyn_hoodieStrape02_l (PPPPPPPP)"!(
    ComponentPose=@n31
    bAlphaBoolEnabled:bool=True
    ConstraintWeights_0=0.0421
    ConstraintWeights_1:float=0.1194
    ConstraintWeights_2:float=0.2936
    ConstraintWeights_3:float=0.3578
    ConstraintWeights_4:float=0.0072
    ConstraintWeights_5:float=0.0001
    ConstraintWeights_6:float=0.1391
    ConstraintWeights_7:float=0.0408
   ) -> (Pose) {
    alphaInputType=Bool
    alphaBoolBlend {blendOption=Linear}
    boneToModify {boneName=dyn_hoodieStrape02_l}
    constraintSetup [
     {targetBone {boneName=spine_02}}
     {targetBone {boneName=spine_03}}
     {targetBone {boneName=spine_04}}
     {targetBone {boneName=spine_05}}
     {targetBone {boneName=deform_upperarm_fr_l}}
     {targetBone {boneName=deform_upperarm_up_l}}
     {targetBone {boneName=deform_pec_l}}
     {targetBone {boneName=clavicle_l}}
    ]
    constraintWeights [1,1,1,1,1,1,1,1]
   }
   n33 = "Constraint - dyn_hoodieStrape03_l (PPPPPPPP)"!(
    ComponentPose=@n32
    bAlphaBoolEnabled:bool=True
    ConstraintWeights_0=0.0001
    ConstraintWeights_1:float=0.0692
    ConstraintWeights_2:float=0.1669
    ConstraintWeights_3:float=0.3402
    ConstraintWeights_4:float=0.2426
    ConstraintWeights_5:float=0.0006
    ConstraintWeights_6:float=0.1804
    ConstraintWeights_7:float=0
   ) -> (Pose) {
    alphaInputType=Bool
    alphaBoolBlend {blendOption=Linear}
    boneToModify {boneName=dyn_hoodieStrape03_l}
    constraintSetup [
     {targetBone {boneName=spine_01}}
     {targetBone {boneName=spine_02}}
     {targetBone {boneName=spine_03}}
     {targetBone {boneName=spine_04}}
     {targetBone {boneName=spine_05}}
     {targetBone {boneName=deform_upperarm_fr_l}}
     {targetBone {boneName=deform_pec_l}}
     {targetBone {boneName=deform_upperarm_fr_r}}
    ]
    constraintWeights [1,1,1,1,1,1,1,1]
   }
   n34 = Anim Dynamics - Bone: dyn_hoodieStrape02_l(ComponentPose=@n33, Alpha=1, ExternalForce="0,0,0") {
    angularDampingOverride=0.99
    angularSpringConstant=75
    bAngularSpring=true
    bOverrideAngularDamping=true
    bUseGravityOverride=true
    simulationSpace=RootRelative
    alphaBoolBlend {blendOption=Linear}
    boundBone {boneName=dyn_hoodieStrape02_l}
    componentAppliedLinearAccClamp=<5000 5000 5000>
    gravityOverride {z=-980}
    physicsBodyDefinitions [{
      boundBone {boneName=dyn_hoodieStrape02_l}
      boxExtents=<10 10 5>
      constraintSetup {angularLimitsMax=<0 0 3>, angularLimitsMin=<0 -15 -3>, angularTarget=<1 0 0>}
      localJointOffset=<0.3 0 0>
     }]
   }
   n35 = Anim Dynamics - Bone: dyn_hoodieStrape03_l(ComponentPose=@n34, Alpha=1, ExternalForce="0,0,0") {
    angularDampingOverride=0.99
    angularSpringConstant=75
    bAngularSpring=true
    bOverrideAngularDamping=true
    bUseGravityOverride=true
    simulationSpace=RootRelative
    alphaBoolBlend {blendOption=Linear}
    boundBone {boneName=dyn_hoodieStrape03_l}
    componentAppliedLinearAccClamp=<5000 5000 5000>
    gravityOverride {z=-980}
    physicsBodyDefinitions [{
      boundBone {boneName=dyn_hoodieStrape03_l}
      boxExtents=<10 10 5>
      constraintSetup {angularLimitsMax=<0 0 3>, angularLimitsMin=<0 -15 -3>, angularTarget=<1 0 0>}
      localJointOffset=<0.3 0 0>
     }]
   }
   n36 = Anim Dynamics - Bone: dyn_hoodieStrape02_r(ComponentPose=@n35, Alpha=1, ExternalForce="0,0,0") {
    angularDampingOverride=0.99
    angularSpringConstant=75
    bAngularSpring=true
    bOverrideAngularDamping=true
    bUseGravityOverride=true
    simulationSpace=RootRelative
    alphaBoolBlend {blendOption=Linear}
    boundBone {boneName=dyn_hoodieStrape02_r}
    componentAppliedLinearAccClamp=<5000 5000 5000>
    gravityOverride {z=-980}
    physicsBodyDefinitions [{
      boundBone {boneName=dyn_hoodieStrape02_r}
      boxExtents=<10 10 5>
      constraintSetup {angularLimitsMax=<0 0 3>, angularLimitsMin=<0 -15 -3>, angularTarget=<1 0 0>}
      localJointOffset=<0.3 0 0>
     }]
   }
   n37 = Anim Dynamics - Bone: dyn_hoodieStrape03_r(ComponentPose=@n36, Alpha=1, ExternalForce="0,0,0") {
    angularDampingOverride=0.99
    angularSpringConstant=75
    bAngularSpring=true
    bOverrideAngularDamping=true
    bUseGravityOverride=true
    simulationSpace=RootRelative
    alphaBoolBlend {blendOption=Linear}
    boundBone {boneName=dyn_hoodieStrape03_r}
    componentAppliedLinearAccClamp=<5000 5000 5000>
    gravityOverride {z=-980}
    physicsBodyDefinitions [{
      boundBone {boneName=dyn_hoodieStrape03_r}
      boxExtents=<10 10 5>
      constraintSetup {angularLimitsMax=<0 0 3>, angularLimitsMin=<0 -15 -3>, angularTarget=<1 0 0>}
      localJointOffset=<0.3 0 0>
     }]
   }
   Save cached pose 'AnimDynamics'(Pose=Component To Local(ComponentPose=@n37)) {cacheName=AnimDynamics}
  ]}
 EventGraph {nodes [
   Event BlueprintUpdateAnimation()
   > Set IsVelocityZero(IsVelocityZero:bool="Equal (Vector)"!(A:Vector=Get PawnVelocity(), B:Vector, ErrorTolerance:float) -> (ReturnValue:bool=~) {is=title K2Node_PromotableOperator}) -> (Output_Get:bool)
   n1 = "integer >= integer"(A=Get CurrentLOD(), B=1)
   > Set IsLOD1AndLower(IsLOD1AndLower:bool=@n1) -> (Output_Get:bool)
   > Set IsLOD1AndLowerBugFix(IsLOD1AndLowerBugFix:double="To Float (Boolean)"(InBool=@n1)) -> (Output_Get:double)
   n2 = Event BlueprintInitializeAnimation()
   n3 = Gravity Override State Function DEPRECATED(
    execute=[@n2.then, @n5.then]
    CharacterPartAnimInstance=Self-Reference()
    GameTimeInSeconds=GetGameTimeInSeconds()
    GravityOverrideParams=Get Gravity Override Params()
   )
   > Set Gravity Override(Gravity Override=@n3)
   n4 = CheckForBackbling()
   > Set Is Backbling A Cape(Is Backbling A Cape:bool=@n4.bIsBackblingACape) -> (Output_Get:bool)
   > Set Is Not Backbling A Cape(Is Not Backbling A Cape:bool=NOT Boolean(A=@n4.bIsBackblingACape)) -> (Output_Get:bool)
   n5 = Event OnCharacterCustomizationLoaded()
   Event OnBeginCrouching()
   > Set IsCrouching(IsCrouching:bool=true) -> (Output_Get:bool)
   Event OnEndCrouching()
   > Set IsCrouching(IsCrouching=false)
   Event OnWeaponChanged()
  ]}
]
variables [
 Gravity Override {type=Vector}
 Gravity Override Params {type=GravityOverrideParamsStruct}
 Is Backbling A Cape {type=boolean}
 Is Not Backbling A Cape {type=boolean}
 IsCrouching {type=boolean}
 IsLOD1AndLower {type=boolean}
 IsLOD1AndLowerBugFix {type=number}
 IsVelocityZero {type=boolean}
]
